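# Physics asset for the belt mesh of the female medium MeekCrow player character, defining the physics bodies for its dangling belt bones.
# Three dyn_belt bones form a capsule chain hanging from the pelvis, with tight swing and twist limits and a second capsule on the last link.
# Larger capsules on pelvis, spine_01 and both thighs approximate the surrounding body, joined to the pelvis by angularly free constraints.
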PhysicsAsset /BRCosmetics/Characters/Player/Female/Medium/Bodies/F_MED_MeekCrow/Meshes/F_MED_MeekCrow_belt_Physics
bodies {
 dyn_belt_01 [{
   shapeType=Capsule
   length=1.706
   radius=1.2611
   shapeName=dyn_belt_01_capsule
   center=<0.8907 1.5069 -0.1801>
   rotation=<rot 6.1185 -95.023 -88.153>
  }]
 dyn_belt_02 [{
   shapeType=Capsule
   length=1.081
   radius=1.3861
   shapeName=dyn_belt_01_capsule
   center=<1.8991 1.2858 0.4533>
   rotation=<rot 4.538 -94.933 88.7246>
  }]
 dyn_belt_03 [
  {
   shapeType=Capsule
   length=1.206
   radius=1.8861
   shapeName=dyn_belt_01_capsule
   center=<2.5989 0.4448 1.1425>
   rotation=<rot 4.5373 -94.9328 -90.8685>
  }
  {
   shapeType=Capsule
   length=2.875
   radius=1.875
   shapeName=dyn_belt_03_capsule
   center=<2.3121 0.8237 1.7081>
   rotation=<rot 90 -179.9999 179.9999>
  }
 ]
 pelvis [{
   shapeType=Capsule
   length=10.1822
   radius=9.3681
   shapeName=pelvis_capsule
   center=<1.7196 -1.3319 0.1473>
   rotation=<rot 7.6758 -93.0805 0.9443>
  }]
 spine_01 [{
   shapeType=Capsule
   length=10.6822
   radius=8.8681
   shapeName=pelvis_capsule
   center=<2.8741 -1.0609 0.0836>
   rotation=<rot 7.6758 -93.0805 0.9443>
  }]
 thigh_l [{
   shapeType=Capsule
   length=27.0922
   radius=7.1487
   shapeName=thigh_l_capsule
   center=<-12.7548 0.0303 0.8752>
   rotation=<rot -0.6579 -96.7919 -92.2433>
  }]
 thigh_r [{
   shapeType=Capsule
   length=27.0922
   radius=7.1487
   shapeName=thigh_r_capsule
   center=<12.7548 -0.0303 -0.8752>
   rotation=<rot -0.6579 -96.7919 87.7567>
  }]
}
constraints [
 {
  bone1Name=thigh_l
  bone2Name=pelvis
  swing1LimitDegrees=45
  swing1Motion=Free
  swing2LimitDegrees=45
  swing2Motion=Free
  twistLimitDegrees=45
  twistMotion=Free
 }
 {
  bone1Name=thigh_r
  bone2Name=pelvis
  swing1LimitDegrees=45
  swing1Motion=Free
  swing2LimitDegrees=45
  swing2Motion=Free
  twistLimitDegrees=45
  twistMotion=Free
 }
 {bone1Name=dyn_belt_01, bone2Name=pelvis, swing1LimitDegrees=5, swing2LimitDegrees=5, twistLimitDegrees=1}
 {bone1Name=dyn_belt_02, bone2Name=dyn_belt_01, swing1LimitDegrees=10, swing2LimitDegrees=10, twistLimitDegrees=2}
 {bone1Name=dyn_belt_03, bone2Name=dyn_belt_02, swing1LimitDegrees=10, swing2LimitDegrees=10, twistLimitDegrees=2}
 {
  bone1Name=spine_01
  bone2Name=pelvis
  swing1LimitDegrees=45
  swing1Motion=Free
  swing2LimitDegrees=45
  swing2Motion=Free
  twistLimitDegrees=45
  twistMotion=Free
 }
]
