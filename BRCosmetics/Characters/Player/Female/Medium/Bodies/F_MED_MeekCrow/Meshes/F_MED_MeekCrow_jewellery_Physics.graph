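# Physics asset for the MeekCrow female medium body's jewellery mesh, defining capsule bodies for a dynamic necklace over the upper torso.
# Two bone chains, dyn_necklace and dyn_necklace_tp, each hang from spine_05 through a MSTR root body.
# Overlapping capsules on spine_05, plus one on spine_04, approximate the torso volume.
# Upper joints rotate freely, while lower links carry narrower swing limits and near-zero twist limits.
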PhysicsAsset /BRCosmetics/Characters/Player/Female/Medium/Bodies/F_MED_MeekCrow/Meshes/F_MED_MeekCrow_jewellery_Physics
bodies {
 dyn_necklace_01 [{shapeType=Capsule, length=1.01, radius=0.5, shapeName=dyn_necklace_01_capsule, rotation=<rot 85 89.9997 -180>}]
 dyn_necklace_01_MSTR [{shapeType=Capsule, length=1.255, radius=0.5, shapeName=dyn_necklace_01_MSTR_capsule, rotation=<rot 0 -90.0006 -90>}]
 dyn_necklace_02 [{
   shapeType=Capsule
   length=13.7078
   radius=0.6171
   shapeName=dyn_necklace_02_capsule
   center=<0.9394 -0.0921 -0.0025>
   rotation=<rot -0.0045 -90 0.0159>
  }]
 dyn_necklace_03 [
  {shapeType=Capsule, length=13.5, radius=1.75, shapeName=dyn_necklace_03_capsule, center=<0.7426 -0.9075 0>}
  {shapeType=Capsule, length=16.5, radius=1.5, shapeName=dyn_necklace_03_capsule, center=<1.9331 -0.6838 0>}
 ]
 dyn_necklace_04 [
  {
   shapeType=Capsule
   length=8.0642
   radius=1.7356
   shapeName=dyn_necklace_04_capsule
   center=<0.9219 -0.8981 0.0376>
   rotation=<rot 0.4539 -90.0003 -0.1752>
  }
  {shapeType=Capsule, length=13, radius=1.75, shapeName=dyn_necklace_04_capsule, center=<-0.9866 -0.9828 0>}
 ]
 dyn_necklace_tp_01 [{shapeType=Capsule, length=13.505, radius=1.005, shapeName=dyn_necklace_tp_01_capsule, center=<2.7901 0.5443 0>, rotation=<rot 0 -90 0>}]
 dyn_necklace_tp_01_MSTR [{shapeType=Capsule, length=0.505, radius=0.505, shapeName=dyn_necklace_tp_01_MSTR_capsule, rotation=<rot 0 -90.0005 90>}]
 dyn_necklace_tp_02 [{
   shapeType=Capsule
   length=7.2502
   radius=1.75
   shapeName=dyn_necklace_tp_02_capsule
   center=<1.8795 -0.6058 0>
   rotation=<rot -0.0064 -90 -0.0029>
  }]
 dyn_necklace_tp_03 [{
   shapeType=Capsule
   length=4.0873
   radius=1.211
   shapeName=dyn_necklace_tp_03_capsule
   center=<0 0.012 0.0004>
   rotation=<rot -0.097 -90 -0.2407>
  }]
 spine_04 [{
   shapeType=Capsule
   length=5.5195
   radius=9.9174
   shapeName=spine_04_capsule
   center=<1.8951 0.5823 0.242>
   rotation=<rot -2.1072 -90 -1.0659>
  }]
 spine_05 [
  {
   shapeType=Capsule
   length=11.386
   radius=6.3079
   shapeName=spine_05_capsule
   center=<1.5901 -0.0641 -0.0272>
   rotation=<rot 0 -173.4679 -179.6403>
  }
  {shapeType=Capsule, length=9.625, radius=6.875, shapeName=spine_05_capsule, center=<-3.2755 3.1783 0.06>}
  {shapeType=Capsule, length=7.5, radius=7.625, shapeName=spine_05_capsule, center=<-2.1298 1.5523 0.2293>}
  {shapeType=Capsule, length=5.5, radius=3.25, shapeName=spine_05_capsule, center=<5.166 1.098 0>, rotation=<rot 90 0 45.4277>}
 ]
}
constraints [
 {
  bone1Name=spine_05
  bone2Name=spine_04
  swing1LimitDegrees=45
  swing1Motion=Free
  swing2LimitDegrees=45
  swing2Motion=Free
  twistLimitDegrees=45
  twistMotion=Free
 }
 {
  bone1Name=dyn_necklace_01_MSTR
  bone2Name=spine_05
  swing1LimitDegrees=45
  swing1Motion=Free
  swing2LimitDegrees=45
  swing2Motion=Free
  twistLimitDegrees=45
  twistMotion=Free
 }
 {
  bone1Name=dyn_necklace_01
  bone2Name=dyn_necklace_01_MSTR
  swing1LimitDegrees=45
  swing1Motion=Free
  swing2LimitDegrees=45
  swing2Motion=Free
  twistLimitDegrees=45
  twistMotion=Free
 }
 {
  bone1Name=dyn_necklace_02
  bone2Name=dyn_necklace_01
  swing1LimitDegrees=45
  swing1Motion=Free
  swing2LimitDegrees=30
  swing2Motion=Free
  twistLimitDegrees=2
  twistMotion=Free
 }
 {bone1Name=dyn_necklace_03, bone2Name=dyn_necklace_02, swing1LimitDegrees=30, swing2LimitDegrees=15, twistLimitDegrees=0}
 {bone1Name=dyn_necklace_04, bone2Name=dyn_necklace_03, swing1LimitDegrees=30, swing2LimitDegrees=5, twistLimitDegrees=0}
 {
  bone1Name=dyn_necklace_tp_01_MSTR
  bone2Name=spine_05
  swing1LimitDegrees=45
  swing1Motion=Free
  swing2LimitDegrees=45
  swing2Motion=Free
  twistLimitDegrees=45
  twistMotion=Free
 }
 {
  bone1Name=dyn_necklace_tp_01
  bone2Name=dyn_necklace_tp_01_MSTR
  swing1LimitDegrees=45
  swing1Motion=Free
  swing2LimitDegrees=45
  swing2Motion=Free
  twistLimitDegrees=45
  twistMotion=Free
 }
 {bone1Name=dyn_necklace_tp_02, bone2Name=dyn_necklace_tp_01, swing1LimitDegrees=20, swing2LimitDegrees=5, twistLimitDegrees=0}
 {bone1Name=dyn_necklace_tp_03, bone2Name=dyn_necklace_tp_02, swing1LimitDegrees=30, swing2LimitDegrees=10, twistLimitDegrees=2}
]
