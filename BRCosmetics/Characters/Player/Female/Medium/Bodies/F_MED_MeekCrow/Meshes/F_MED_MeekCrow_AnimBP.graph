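# Post-process style Anim Blueprint for the MeekCrow female body, driving secondary motion on its necklace and belt.
# Copied parent-mesh pose runs through a Control Rig, then spine-based jewellery and pelvis-based belt RigidBody simulations; only the belt uses computed gravity override.
# Additive necklace bone offsets then correct for moving, airborne, emote and skydive/parachute states.
# A backbling flag is set on customization load but unused in the AnimGraph.
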AnimBlueprint /BRCosmetics/Characters/Player/Female/Medium/Bodies/F_MED_MeekCrow/Meshes/F_MED_MeekCrow_AnimBP
target_skeleton=&/F_MED_MeekCrow_Skeleton
graphs [
 AnimGraph {nodes [
   n1 = Get skydive floating() -> (skydive floating:double)
   n2 = "Transform (Modify) Bone - Bone: dyn_necklace_tp_02"(Alpha=@n1, Translation="0,-2,0", Rotation="0,0,0", Scale="1,1,1") {
    translationMode=BMM_Additive
    translationSpace=BCS_ParentBoneSpace
    alphaBoolBlend {blendOption=Linear}
    boneToModify {boneName=dyn_necklace_tp_02}
    translation {y=-2}
   }
   "Transform (Modify) Bone - Bone: dyn_necklace_02"(ComponentPose=@n2, Alpha=@n1, Translation="0,-2,0", Rotation="0,0,0", Scale="1,1,1") {
    translationMode=BMM_Additive
    translationSpace=BCS_ParentBoneSpace
    alphaBoolBlend {blendOption=Linear}
    boneToModify {boneName=dyn_necklace_02}
    translation {y=-2}
   }
   n3 = Control Rig(Source=Copy Pose From Mesh() {bUseAttachedParent=true}) {
    alphaBoolBlend {blendOption=Linear}
    controlRigAssetReference {blueprintRigClass=&/F_MED_MeekCrow_CtrlRig.F_MED_MeekCrow_CtrlRig_C}
   }
   n4 = RigidBody!(ComponentPose=Local To Component(LocalPose=@n3), Alpha, OverrideWorldGravity:Vector="0,0,-980", ExternalForce) -> (Pose) {
    bOverrideWorldGravity=true
    overridePhysicsAsset=&/F_MED_MeekCrow_jewellery_Physics
    simulationSpace=BaseBoneSpace
    alphaBoolBlend {blendOption=Linear}
    baseBoneRef {boneName=spine_05}
    simSpaceSettings {maxAngularAcceleration=200, maxAngularVelocity=40, maxLinearAcceleration=2000, maxLinearVelocity=200, worldAlpha=1}
   }
   n5 = RigidBody(ComponentPose=@n4, OverrideWorldGravity=Get GravityOverride()) {
    overridePhysicsAsset=&/F_MED_MeekCrow_belt_Physics
    simulationSpace=BaseBoneSpace
    alphaBoolBlend {blendOption=Linear}
    baseBoneRef {boneName=pelvis}
    simSpaceSettings {maxAngularAcceleration=80, maxAngularVelocity=16, maxLinearAcceleration=2000, maxLinearVelocity=400, worldAlpha=1}
   }
   n6 = "Transform (Modify) Bone - Bone: dyn_necklace_tp_01"!(ComponentPose=@n5, bAlphaBoolEnabled:bool=Get bIsMoving2D(), Translation="0,-0.6,0", Rotation="0,0,0", Scale="1,1,1") -> (Pose) {
    alphaInputType=Bool
    rotationMode=BMM_Additive
    rotationSpace=BCS_BoneSpace
    translationMode=BMM_Additive
    translationSpace=BCS_BoneSpace
    alphaBoolBlend {blendOption=Linear}
    boneToModify {boneName=dyn_necklace_tp_01}
    translation {y=-0.6}
   }
   n7 = "Transform (Modify) Bone - Bone: dyn_necklace_tp_01"!(ComponentPose=@n6, Alpha:float=@n1, Translation="0,-3,0", Rotation="0,-20,0", Scale="1,1,1") -> (Pose) {
    rotationMode=BMM_Additive
    translationMode=BMM_Additive
    translationSpace=BCS_BoneSpace
    alphaBoolBlend {blendOption=Linear}
    boneToModify {boneName=dyn_necklace_tp_01}
    rotation {yaw=-20}
    translation {y=-3}
   }
   n8 = "Transform (Modify) Bone - Bone: dyn_necklace_01"(ComponentPose=@n7, Alpha=@n1, Translation="0,-1,0", Rotation="0,-15,0", Scale="1,1,1") {
    rotationMode=BMM_Additive
    rotationSpace=BCS_BoneSpace
    translationMode=BMM_Additive
    translationSpace=BCS_BoneSpace
    alphaBoolBlend {blendOption=Linear}
    boneToModify {boneName=dyn_necklace_01}
    rotation {yaw=-15}
    translation {y=-1}
   }
   n9 = "Transform (Modify) Bone - Bone: dyn_necklace_03"!(ComponentPose=@n8, bAlphaBoolEnabled:bool=Get bIsInAir() -> (bIsInAir:bool), Translation="0,-0.6,0", Rotation="0,-1,0", Scale="1,1,1") -> (Pose) {
    alphaInputType=Bool
    rotationMode=BMM_Additive
    rotationSpace=BCS_BoneSpace
    translationMode=BMM_Additive
    translationSpace=BCS_BoneSpace
    alphaBoolBlend {blendOption=Linear}
    boneToModify {boneName=dyn_necklace_03}
    rotation {yaw=-1}
    translation {y=-0.6}
   }
   n10 = Get bIsPlayingEmote()
   n11 = "Transform (Modify) Bone - Bone: dyn_necklace_01"(ComponentPose=@n9, Alpha=1, Translation="0,0,0", Rotation="0,-2,0", Scale="1,1,1", None=@n10) {
    rotationMode=BMM_Additive
    rotationSpace=BCS_BoneSpace
    scaleSpace=BCS_BoneSpace
    translationMode=BMM_Additive
    translationSpace=BCS_BoneSpace
    alphaBoolBlend {blendOption=Linear}
    boneToModify {boneName=dyn_necklace_01}
    rotation {yaw=-2}
   }
   n12 = "Transform (Modify) Bone - Bone: dyn_necklace_03"(ComponentPose=@n11, bAlphaBoolEnabled=@n10, Translation="0,0,0", Rotation="0,-2,0", Scale="1,1,1") {
    alphaInputType=Bool
    rotationMode=BMM_Additive
    rotationSpace=BCS_BoneSpace
    scaleSpace=BCS_BoneSpace
    translationMode=BMM_Additive
    translationSpace=BCS_BoneSpace
    alphaBoolBlend {blendOption=Linear}
    boneToModify {boneName=dyn_necklace_03}
    rotation {yaw=-2}
   }
   n13 = "Transform (Modify) Bone - Bone: dyn_necklace_tp_01"!(ComponentPose=@n12, bAlphaBoolEnabled:bool=@n10, Translation="0,0,0", Rotation="0,-2,0", Scale="1,1,1") -> (Pose) {
    alphaInputType=Bool
    rotationMode=BMM_Additive
    rotationSpace=BCS_BoneSpace
    translationMode=BMM_Additive
    translationSpace=BCS_BoneSpace
    alphaBoolBlend {blendOption=Linear}
    boneToModify {boneName=dyn_necklace_tp_01}
    rotation {yaw=-2}
   }
   Output Pose(Result=Component To Local(ComponentPose=@n13)) {layerGroup=None}
  ]}
 EventGraph {nodes [
   Event BlueprintUpdateAnimation()
   > n1 = Sequence()
   n2 = Get bIsSkydiveFloating()
   n3 = Get bIsParachuteOpened()
   > Set skydive floating(skydive floating:double="To Float (Boolean)"(InBool=OR Boolean(A=@n2, B=@n3))) -> (Output_Get:double)
   n4 = Gravity Override State Function DEPRECATED(
    execute=@n1.then_1
    CharacterPartAnimInstance=Self-Reference()
    GameTimeInSeconds=GetGameTimeInSeconds()
    GravityOverrideParams=Get Gravity Override Params()
   )
   > Set GravityOverride(GravityOverride=@n4)
   Event OnCharacterCustomizationLoaded()
   > Do Once()
   n5 = CheckForBackbling()
   > Set isWearingBackbling(isWearingBackbling=OR Boolean(A=@n5, B=@n5.bIsBackblingLong))
   n6 = GetPartSkeletalMeshComponent()
   "Not Equal (Object)"!(
    A:Object Reference=FindChildCommponentFromAttachSocket!(
     self:Custom Character Part Anim Instance Object Reference
     Parent:Skeletal Mesh Component Object Reference=@n6
     bIncludeAllDescendants:bool=false
     AttachSocket:Name=cape
    ) -> (ReturnValue:Skeletal Mesh Component Object Reference)
    B:Object Reference
   ) -> (ReturnValue:bool)
   "Not Equal (Object)"(A=FindChildCommponentFromAttachSocket(Parent=@n6, bIncludeAllDescendants=false, AttachSocket=Backpack_BR))
  ]}
]
variables [
 Gravity Override Params {type=GravityOverrideParamsStruct}
 GravityOverride {type=Vector}
 isWearingBackbling {type=boolean}
 skydive floating {type=number}
]
