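# Control Rig for the MeekCrow face accessory part that adjusts the dynamic back-hair bones against the head and upper body.
# The first sequence branch runs left and right upper-arm pose readers, each followed by a mirrored hair rotation at fixed weight, with reader outputs unconnected.
# It then aims the hair base at a repositioned helper bone under spine_05.
# The second branch parent-constrains that base between spine_05 and head, favoring the spine.
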ControlRigBlueprint /BRCosmetics/Characters/Player/Female/Medium/Bodies/F_MED_MeekCrow/Meshes/Parts/F_MED_MeekCrow_FaceAcc_CtrlRig
graphs [
 RigVMFunctionLibrary {links [], nodes []}
 RigVMModel {nodes [
   RigUnit_BeginExecution = &RigUnit_BeginExecution()
   RigVMFunction_Sequence = &RigVMFunction_Sequence(ExecuteContext=@RigUnit_BeginExecution.ExecutePin)
   &RigUnit_SphericalPoseReader(
    ActiveRegionScaleFactors=(PositiveWidth=1,NegativeWidth=1,PositiveHeight=1,NegativeHeight=1)
    ActiveRegionSize=0.1
    Debug=(bDrawDebug=True,bDraw2D=False,bDrawLocalAxes=False,DebugScale=25,DebugSegments=20,DebugThickness=0.25)
    DriverAxis=(X=1,Y=0,Z=0)
    DriverItem=(Type=Bone,Name="upperarm_l")
    FalloffRegionScaleFactors=(PositiveWidth=1,NegativeWidth=1,PositiveHeight=1,NegativeHeight=1)
    FalloffSize=0.3
    FlipHeightScaling=False
    FlipWidthScaling=False
    OptionalParentItem=(Type=Bone,Name="None")
    RotationOffset=(X=0,Y=0,Z=0)
    ExecutePin=@RigVMFunction_Sequence.A
   )
   > &RigUnit_ModifyTransforms(
    ItemToModify="((Item=(Type=Bone,Name=\"dyn_b_base_hair\"),Transform=(Rotation=(X=0,Y=0,Z=0.1736,W=0.9848),Translation=(X=0,Y=0,Z=0),Scale3D=(X=1,Y=1,Z=1))),(Item=(Type=Bone,Name=\"dyn_b_hair_3\"),Transform=(Rotation=(X=0,Y=0,Z=0.0872,W=0.9962),Translation=(X=0,Y=0,Z=0),Scal ...(truncated)"
    Weight=1
    WeightMinimum=0
   )
   > &RigUnit_SphericalPoseReader(
    ActiveRegionScaleFactors=(PositiveWidth=1,NegativeWidth=1,PositiveHeight=1,NegativeHeight=1)
    ActiveRegionSize=0.1
    Debug=(bDrawDebug=True,bDraw2D=False,bDrawLocalAxes=False,DebugScale=25,DebugSegments=20,DebugThickness=0.25)
    DriverAxis=(X=-1,Y=0,Z=0)
    DriverItem=(Type=Bone,Name="upperarm_r")
    FalloffRegionScaleFactors=(PositiveWidth=1,NegativeWidth=1,PositiveHeight=1,NegativeHeight=1)
    FalloffSize=0.3
    FlipHeightScaling=False
    FlipWidthScaling=False
    OptionalParentItem=(Type=Bone,Name="None")
    RotationOffset=(X=0,Y=180,Z=0)
   )
   > &RigUnit_ModifyTransforms(
    ItemToModify="((Item=(Type=Bone,Name=\"dyn_b_base_hair\"),Transform=(Rotation=(X=0,Y=0,Z=-0.1736,W=0.9848),Translation=(X=0,Y=0,Z=0),Scale3D=(X=1,Y=1,Z=1))),(Item=(Type=Bone,Name=\"dyn_b_hair_3\"),Transform=(Rotation=(X=0,Y=0,Z=-0.0872,W=0.9962),Translation=(X=0,Y=0,Z=0),Sc ...(truncated)"
    Weight=1
    WeightMinimum=0
   )
   > &RigUnit_SetTransform(
    Item=(Type=Bone,Name="hair_aim_const")
    Space=LocalSpace
    Value=(Rotation=(X=0,Y=0,Z=0,W=1),Translation=(X=0,Y=3,Z=0),Scale3D=(X=1,Y=1,Z=1))
    Weight=1
   )
   > /Script/ControlRig.RigUnit_AimItem(
    DebugSettings=(bEnabled=False,Scale=10,WorldOffset=(Rotation=(X=0,Y=0,Z=0,W=1),Translation=(X=0,Y=0,Z=0),Scale3D=(X=1,Y=1,Z=1)))
    Item=(Type=Bone,Name="dyn_b_base_hair")
    Primary=(Weight=1,Axis=(X=0,Y=0,Z=1),Target=(X=10,Y=0,Z=0),Kind=Location,Space=(Type=Bone,Name="hair_aim_const"))
    Secondary=(Weight=1,Axis=(X=0,Y=0,Z=0),Target=(X=1,Y=0,Z=0),Kind=Direction,Space=(Type=Bone,Name="head"))
    Weight=1
   ) {is=struct RigVMUnitNode}
   &RigUnit_ParentConstraint(
    AdvancedSettings=(InterpolationType=Average,RotationOrderForFilter=XZY)
    Child=(Type=Bone,Name="dyn_b_base_hair")
    Parents=((Item=(Type=Bone,Name="spine_05"),Weight=0.6),(Item=(Type=Bone,Name="head"),Weight=0.4))
    Weight=1
    ExecutePin=@RigVMFunction_Sequence.B
   )
  ]}
]
hierarchy [
 root {loc=<0 0 0>, rot=<0 0 0 1>}
 pelvis {parent=root, loc=<0 -0.5301 98.0571>, rot=<0.0127 0.707 -0.0127 -0.707>}
 spine_01 {parent=pelvis, loc=<6.1208 -0.9328 0>, rot=<0 0 0.034 -0.9994>}
 spine_02 {parent=spine_01, loc=<6.5997 -0.2095 0>, rot=<0 0 0.0238 -0.9997>}
 spine_03 {parent=spine_02, loc=<6.4914 -0.1237 0>, rot=<0 0 0.0149 -0.9999>}
 spine_04 {parent=spine_03, loc=<6.9598 0.0794 0>, rot=<0 0 -0.0132 -0.9999>}
 spine_05 {parent=spine_04, loc=<6.9987 0 0>, rot=<0 0 -0.0015 -1>}
 neck_01 {parent=spine_05, loc=<11.3138 0.1529 0>, rot=<0 0 -0.2003 -0.9797>}
 neck_02 {parent=neck_01, loc=<4.5627 0 0>, rot=<0 0 0.0504 -0.9987>}
 head {parent=neck_02, loc=<5.3914 0 0>, rot=<0 0 0.115 -0.9934>}
 head_offset {parent=head, loc=<0 0 0>, rot=<0 0 0 1>}
 dyn_Hat_base {parent=head_offset, loc=<6.5738 -7.8155 1.5929>, rot=<0.2556 -0.9539 0.0406 0.1517>}
 dyn_Hat_1 {parent=dyn_Hat_base, loc=<0 0 0>, rot=<0 0 0 1>}
 dyn_Hat_2 {parent=dyn_Hat_1, loc=<3.8309 0 0>, rot=<-0.0126 -0.0747 0.166 -0.9832>}
 dyn_l_base_hair {parent=head_offset, loc=<9.4193 5.6657 -4>, rot=<-0.0009 -0.9411 -0.2153 -0.2606>}
 dyn_l_hair_1 {parent=dyn_l_base_hair, loc=<0 0 0>, rot=<0 0 0 1>}
 dyn_l_hair_2 {parent=dyn_l_hair_1, loc=<4.5 0 0>, rot=<-0.0068 -0.1881 -0.0355 -0.9815>}
 dyn_l_hair_3 {parent=dyn_l_hair_2, loc=<4.758 0 0>, rot=<0 -0.094 0 -0.9956>}
 dyn_l_hair_4 {parent=dyn_l_hair_3, loc=<4.0302 0 0>, rot=<0 -0.0905 0 -0.9959>}
 dyn_b_base_hair {parent=head_offset, loc=<-4.0599 -5.3991 0>, rot=<-0.2103 0.6751 0.6751 -0.2103>}
 dyn_b_hair_1 {parent=dyn_b_base_hair, loc=<0 0 0>, rot=<0 0 0 1>}
 dyn_b_hair_2 {parent=dyn_b_hair_1, loc=<7.3536 0 0>, rot=<0 0.1364 0 -0.9907>}
 dyn_b_hair_3 {parent=dyn_b_hair_2, loc=<5.0255 0 0>, rot=<0 0.0482 0 -0.9988>}
 dyn_b_hair_4 {parent=dyn_b_hair_3, loc=<5 0 0>, rot=<0 0.0854 0 -0.9963>}
 dyn_b_hair_5 {parent=dyn_b_hair_4, loc=<5 0 0>, rot=<0 0.1111 0 -0.9938>}
 clavicle_l {parent=spine_05, loc=<5.4831 -0.1053 -3.2946>, rot=<0.0465 -0.7289 -0.009 -0.6829>}
 upperarm_l {parent=clavicle_l, loc=<9.8658 -1.2136 -0.4105>, rot=<0.0164 -0.3245 -0.0269 -0.9454>}
 upperarm_twist_01_l {parent=upperarm_l, loc=<7.2135 -0.1134 0>, rot=<0 0 0 1>}
 clavicle_r {parent=spine_05, loc=<5.4831 -0.1053 3.2946>, rot=<0.7289 0.0465 -0.6829 0.009>}
 upperarm_r {parent=clavicle_r, loc=<-9.8658 1.2136 0.4105>, rot=<0.0164 -0.3245 -0.0269 -0.9454>}
 upperarm_twist_01_r {parent=upperarm_r, loc=<-7.2135 0.1134 0>, rot=<0 0 0 -1>}
 hair_aim_const {parent=spine_05, loc=<0 -6.5 0>, rot=<0 0 0 1>}
]
variables [isCrouching {type=double, default=0}]
