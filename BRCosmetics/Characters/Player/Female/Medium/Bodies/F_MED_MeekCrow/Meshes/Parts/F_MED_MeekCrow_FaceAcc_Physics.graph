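# Physics asset for the MeekCrow female medium character's face accessory mesh, set up for secondary motion on its hat and hair.
# Capsules on upper spine, clavicles, neck and head form the body volume around the dynamic bones.
# A two-segment hat chain and five-segment hair chain hang from small base bodies locked to the head.
# Chain joints use limited swing with almost no twist, while head-to-neck and spine_05-to-spine_04 rotate freely.
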PhysicsAsset /BRCosmetics/Characters/Player/Female/Medium/Bodies/F_MED_MeekCrow/Meshes/Parts/F_MED_MeekCrow_FaceAcc_Physics
bodies {
 clavicle_l [{
   shapeType=Capsule
   length=33.2428
   radius=8.4036
   shapeName=clavicle_l_capsule
   center=<13.8311 1.6076 -2.8663>
   rotation=<rot -0.0002 -90 87.683>
  }]
 clavicle_r [{
   shapeType=Capsule
   length=33.2428
   radius=8.4036
   shapeName=clavicle_r_capsule
   center=<-18.1908 -1.6076 3.08>
   rotation=<rot -0.0002 -90 -92.317>
  }]
 dyn_Hat_1 [{
   shapeType=Capsule
   length=6.0624
   radius=2.7385
   shapeName=dyn_Hat_1_capsule
   center=<-0.0463 0.1923 -0.6394>
   rotation=<rot -16.5233 -99.331 176.0134>
  }]
 dyn_Hat_2 [{
   shapeType=Capsule
   length=5.9584
   radius=2.4621
   shapeName=dyn_Hat_2_capsule
   center=<1.837 0.0559 0.2155>
   rotation=<rot -18.1927 -85.7178 -12.6055>
  }]
 dyn_Hat_base [{shapeType=Capsule, length=0.505, radius=0.505, shapeName=dyn_Hat_base_capsule, rotation=<rot 0 -90 0>}]
 dyn_b_base_hair [{shapeType=Sphere, radius=0.505, shapeName=dyn_b_base_hair_sphere}]
 dyn_b_hair_1 [{shapeType=Capsule, length=11.4209, radius=3.5, shapeName=dyn_b_hair_1_capsule, center=<1.6 0.4778 -0.9978>, rotation=<rot 0 -180 -90>}]
 dyn_b_hair_2 [{shapeType=Capsule, length=6.3849, radius=2.2923, shapeName=dyn_b_hair_2_capsule, center=<-0.0405 0 -0.5288>, rotation=<rot 0 -180 -90>}]
 dyn_b_hair_3 [{
   shapeType=Capsule
   length=5.0188
   radius=2.5
   shapeName=dyn_b_hair_3_capsule
   center=<-0.2888 -0.0283 -1.6529>
   rotation=<rot 87.2435 159.5617 -109.2378>
  }]
 dyn_b_hair_4 [{
   shapeType=Capsule
   length=5.8261
   radius=3.1506
   shapeName=dyn_b_hair_4_capsule
   center=<0.3852 -0.0744 -1.8597>
   rotation=<rot 0 178.5817 -89.3265>
  }]
 dyn_b_hair_5 [{
   shapeType=Capsule
   length=5.5006
   radius=3.0484
   shapeName=dyn_b_hair_5_capsule
   center=<0.946 -0.1976 -1.0677>
   rotation=<rot 7.3127 177.2906 -87.1359>
  }]
 head [{
   shapeType=Capsule
   length=1.8437
   radius=5.3239
   shapeName=head_capsule
   center=<2.9103 -0.4936 2.6851>
   rotation=<rot -73.6769 -148.5674 2.9051>
  }]
 neck_01 [{
   shapeType=Capsule
   length=4.9287
   radius=5.2414
   shapeName=neck_01_capsule
   center=<-3.9586 1.4942 0.9267>
   rotation=<rot -1.7798 -73.3823 95.9411>
  }]
 spine_04 [{
   shapeType=Capsule
   length=49.2004
   radius=8.4046
   shapeName=spine_04_capsule
   center=<-2.2568 -0.7315 0.0341>
   rotation=<rot 0 -90.0047 0.8652>
  }]
 spine_05 [{
   shapeType=Capsule
   length=48.6246
   radius=10.352
   shapeName=spine_05_capsule
   center=<-0.5247 1.9357 0.0099>
   rotation=<rot -0.2888 -86.0314 -0.7606>
  }]
}
constraints [
 {
  bone1Name=dyn_Hat_base
  bone2Name=head
  swing1LimitDegrees=45
  swing1Motion=Locked
  swing2LimitDegrees=45
  swing2Motion=Locked
  twistLimitDegrees=45
  twistMotion=Locked
 }
 {bone1Name=dyn_Hat_1, bone2Name=dyn_Hat_base, swing1LimitDegrees=25, swing2LimitDegrees=15, twistLimitDegrees=2}
 {bone1Name=dyn_Hat_2, bone2Name=dyn_Hat_1, swing1LimitDegrees=30, swing2LimitDegrees=20, twistLimitDegrees=2}
 {
  bone1Name=neck_01
  bone2Name=spine_05
  swing1LimitDegrees=45
  swing1Motion=Locked
  swing2LimitDegrees=45
  swing2Motion=Locked
  twistLimitDegrees=45
  twistMotion=Locked
 }
 {
  bone1Name=clavicle_l
  bone2Name=spine_05
  swing1LimitDegrees=45
  swing1Motion=Locked
  swing2LimitDegrees=45
  swing2Motion=Locked
  twistLimitDegrees=45
  twistMotion=Locked
 }
 {
  bone1Name=spine_05
  bone2Name=spine_04
  swing1LimitDegrees=45
  swing1Motion=Free
  swing2LimitDegrees=45
  swing2Motion=Free
  twistLimitDegrees=45
  twistMotion=Free
 }
 {
  bone1Name=clavicle_r
  bone2Name=spine_05
  swing1LimitDegrees=45
  swing1Motion=Locked
  swing2LimitDegrees=45
  swing2Motion=Locked
  twistLimitDegrees=45
  twistMotion=Locked
 }
 {
  bone1Name=dyn_b_base_hair
  bone2Name=head
  swing1LimitDegrees=45
  swing1Motion=Locked
  swing2LimitDegrees=45
  swing2Motion=Locked
  twistLimitDegrees=45
  twistMotion=Locked
 }
 {bone1Name=dyn_b_hair_1, bone2Name=dyn_b_base_hair, swing1LimitDegrees=5, swing2LimitDegrees=18, twistLimitDegrees=0}
 {bone1Name=dyn_b_hair_3, bone2Name=dyn_b_hair_2, swing1LimitDegrees=5, swing2LimitDegrees=45, twistLimitDegrees=0}
 {bone1Name=dyn_b_hair_4, bone2Name=dyn_b_hair_3, swing1LimitDegrees=10, swing2LimitDegrees=45, twistLimitDegrees=0}
 {bone1Name=dyn_b_hair_5, bone2Name=dyn_b_hair_4, swing1LimitDegrees=15, swing2LimitDegrees=30, twistLimitDegrees=0}
 {
  bone1Name=head
  bone2Name=neck_01
  swing1LimitDegrees=45
  swing1Motion=Free
  swing2LimitDegrees=45
  swing2Motion=Free
  twistLimitDegrees=45
  twistMotion=Free
 }
 {bone1Name=dyn_b_hair_2, bone2Name=dyn_b_hair_1, swing1LimitDegrees=5, swing2LimitDegrees=35, twistLimitDegrees=2}
]
