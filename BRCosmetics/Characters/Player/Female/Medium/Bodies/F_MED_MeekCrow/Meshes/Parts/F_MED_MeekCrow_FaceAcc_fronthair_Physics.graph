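# Physics asset for the front-hair part of the female medium MeekCrow character, set up for secondary motion on a left-side hair strand.
# Capsule bodies form a five-bone chain from the head, which carries overlapping spheres and capsules as its collision volume.
# Constraints lock the base and first segment to the head, then allow progressively wider swing toward the tip with almost no twist.
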PhysicsAsset /BRCosmetics/Characters/Player/Female/Medium/Bodies/F_MED_MeekCrow/Meshes/Parts/F_MED_MeekCrow_FaceAcc_fronthair_Physics
bodies {
 dyn_l_base_hair [{shapeType=Capsule, length=0.505, radius=0.505, shapeName=dyn_l_base_hair_capsule, rotation=<rot 0 -90 0>}]
 dyn_l_hair_1 [{
   shapeType=Capsule
   length=0.328
   radius=1.1065
   shapeName=dyn_l_hair_1_capsule
   center=<0.2367 0.2343 0.0661>
   rotation=<rot 0 -13.4462 -80.9145>
  }]
 dyn_l_hair_2 [{
   shapeType=Capsule
   length=2.5007
   radius=1.1698
   shapeName=dyn_l_hair_2_capsule
   center=<2.3969 0.0629 0.5297>
   rotation=<rot 0.091 -85.4481 -89.3712>
  }]
 dyn_l_hair_3 [{
   shapeType=Capsule
   length=1.9373
   radius=1.1141
   shapeName=dyn_l_hair_3_capsule
   center=<1.7169 0.0665 0.3909>
   rotation=<rot -28.7739 -89.9873 87.3375>
  }]
 dyn_l_hair_4 [{
   shapeType=Capsule
   length=1.6114
   radius=1.0122
   shapeName=dyn_l_hair_4_capsule
   center=<1.1548 -0.5636 0.2698>
   rotation=<rot -2.6598 83.4302 -92.7932>
  }]
 head [
  {shapeType=Sphere, radius=7.1001, shapeName=head_sphere, center=<-0.8484 2.4876 0.6919>}
  {shapeType=Sphere, radius=5.6229, shapeName=head_sphere, center=<-2.3008 1.4964 -0.3981>}
  {shapeType=Capsule, length=0.9731, radius=6.6418, center=<-0.1329 4.2961 0.1764>, rotation=<rot 0.9137 -16.0352 -91.7991>}
  {shapeType=Capsule, length=5.25, radius=2.25, shapeName=head_capsule, center=<-2.6649 2.6548 -6.905>, rotation=<rot -90 -172.875 -167.1252>}
 ]
}
constraints [
 {
  bone1Name=dyn_l_base_hair
  bone2Name=head
  swing1LimitDegrees=45
  swing1Motion=Locked
  swing2LimitDegrees=45
  swing2Motion=Locked
  twistLimitDegrees=45
  twistMotion=Locked
 }
 {
  bone1Name=dyn_l_hair_1
  bone2Name=dyn_l_base_hair
  swing1LimitDegrees=45
  swing1Motion=Locked
  swing2LimitDegrees=45
  swing2Motion=Locked
  twistLimitDegrees=45
  twistMotion=Locked
 }
 {bone1Name=dyn_l_hair_2, bone2Name=dyn_l_hair_1, swing1LimitDegrees=5, swing2LimitDegrees=10, twistLimitDegrees=1}
 {bone1Name=dyn_l_hair_3, bone2Name=dyn_l_hair_2, swing1LimitDegrees=10, swing2LimitDegrees=15, twistLimitDegrees=1}
 {bone1Name=dyn_l_hair_4, bone2Name=dyn_l_hair_3, swing1LimitDegrees=20, swing2LimitDegrees=20, twistLimitDegrees=0}
]
