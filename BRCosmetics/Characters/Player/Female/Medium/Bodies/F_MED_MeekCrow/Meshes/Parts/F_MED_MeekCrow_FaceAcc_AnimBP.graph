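# Animation Blueprint for the MeekCrow female character's face accessory part, animating its hair and ponytail.
# The main path layers body and base-mesh poses through Control Rig, front-hair RigidBody, crouch, jump and skydive bone offsets, then a second RigidBody.
# When a back bling is worn, it switches to a cached pose using constraints and AnimDynamics on the ponytail.
# The event graph updates gravity override and movement-state values, and detects back bling on customization load.
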AnimBlueprint /BRCosmetics/Characters/Player/Female/Medium/Bodies/F_MED_MeekCrow/Meshes/Parts/F_MED_MeekCrow_FaceAcc_AnimBP
target_skeleton=&/F_MED_MeekCrow_FaceAcc_Skeleton
graphs [
 AnimGraph {nodes [
   n1 = RigidBody!(
    ComponentPose=Local To Component(LocalPose=Use cached pose 'HairCache'())
    Alpha
    OverrideWorldGravity:Vector=Get GravityOverride()
    ExternalForce
   ) -> (Pose) {
    bOverrideWorldGravity=true
    overridePhysicsAsset=&/F_MED_MeekCrow_FaceAcc_Physics
    simulationSpace=BaseBoneSpace
    alphaBoolBlend {blendOption=Linear}
    baseBoneRef {boneName=spine_05}
    overrideWorldGravity {z=-980}
    simSpaceSettings {maxAngularAcceleration=80, maxAngularVelocity=12, maxLinearAcceleration=2500, maxLinearVelocity=450, worldAlpha=0.8}
   }
   n2 = "Transform (Modify) Bone - Bone: dyn_b_hair_2"(ComponentPose=@n1, Alpha=1, Translation="0,0,3", Rotation="0,0,0", Scale="1,1,1") {translationMode=BMM_Additive, alphaBoolBlend {blendOption=Linear}, boneToModify {boneName=dyn_b_hair_2}, translation {z=3}}
   n3 = Component To Local(ComponentPose=@n2)
   n4 = Use cached pose 'HairNoDYN'()
   n5 = Get isWearingPack() -> (isWearingPack:bool)
   n6 = "Transform (Modify) Bone - Bone: dyn_b_hair_3"!(
    ComponentPose=Local To Component(LocalPose=Blend Poses by bool(BlendPose_0=@n4, BlendPose_1=@n3, bActiveValue=@n5) {blendType=Linear})
    bAlphaBoolEnabled:bool=Get bIsInFrontEnd()
    Translation="0,0,-2.5"
    Rotation="0,0,5"
    Scale="1,1,1"
   ) -> (Pose) {
    alphaInputType=Bool
    rotationMode=BMM_Additive
    translationMode=BMM_Additive
    alphaBoolBlend {blendOption=Linear}
    boneToModify {boneName=dyn_b_hair_3}
    rotation {roll=5}
    translation {z=-2.5}
   }
   Output Pose(Result=Component To Local(ComponentPose=@n6)) {layerGroup=None}
   n7 = Layered blend per bone(
    BasePose=Copy Pose From Mesh(SourceMeshComponent=Get Skel_Mesh()) {bUseAttachedParent=true}
    BlendPoses_0=Copy Pose From Mesh(SourceMeshComponent=Get Body_Mesh()) {bUseAttachedParent=true}
   ) {layerSetup [{branchFilters [{boneName=clavicle_l}, {boneName=clavicle_r}, {boneName=calf_l}, {boneName=calf_r}]}]}
   n8 = Control Rig(Source=@n7) {
    alphaBoolBlend {blendOption=Linear}
    controlRigAssetReference {blueprintRigClass=&/F_MED_MeekCrow_FaceAcc_CtrlRig.F_MED_MeekCrow_FaceAcc_CtrlRig_C}
   }
   Save cached pose 'PonytailBlendPose'(Pose=@n8) {cacheName=PonytailBlendPose}
   n9 = Control Rig!(Source=Use cached pose 'PonytailBlendPose'(), Alpha=0.3, isCrouching:double=Get isCrouching() -> (isCrouching:double)) -> (Pose) {
    alphaBoolBlend {blendOption=Linear}
    controlRigAssetReference {blueprintRigClass=&F_MED_MeekCrow_FaceAcc_CtrlRig_C}
   }
   n10 = Control Rig!(Source=Use cached pose 'PonytailBlendPose'(), Alpha) -> (Pose) {
    alphaBoolBlend {blendOption=Linear}
    controlRigAssetReference {blueprintRigClass=&F_MED_MeekCrow_FaceAcc_CtrlRig_C}
   }
   n11 = Get bIsCrouching()
   n12 = RigidBody(
    ComponentPose=Local To Component(LocalPose=Blend Poses by bool(BlendPose_0=@n10, BlendPose_1=@n9, bActiveValue=@n11) {blendType=Linear})
    OverrideWorldGravity=Get GravityOverride()
   ) {
    bOverrideWorldGravity=true
    overridePhysicsAsset=&/F_MED_MeekCrow_FaceAcc_fronthair_Physics
    simulationSpace=BaseBoneSpace
    alphaBoolBlend {blendOption=Linear}
    baseBoneRef {boneName=head}
    overrideWorldGravity {z=-980}
    simSpaceSettings {maxAngularAcceleration=40, maxAngularVelocity=8, maxLinearAcceleration=1000, maxLinearVelocity=100, worldAlpha=1}
   }
   n13 = "Transform (Modify) Bone - Bone: dyn_b_hair_1"!(ComponentPose=@n12, bAlphaBoolEnabled:bool=Get bIsCrouching(), Translation="0,0,-1", Rotation="0,0,0", Scale="1,1,1") -> (Pose) {
    alphaInputType=Bool
    rotationMode=BMM_Additive
    rotationSpace=BCS_ParentBoneSpace
    translationMode=BMM_Additive
    translationSpace=BCS_BoneSpace
    alphaBoolBlend {blendOption=Linear}
    boneToModify {boneName=dyn_b_hair_1}
    translation {z=-1}
   }
   n14 = "Transform (Modify) Bone - Bone: dyn_b_hair_1"(ComponentPose=@n13, bAlphaBoolEnabled=Get bIsJumping(), Translation="0,0,0", Rotation="0,10,0", Scale="1,1,1") {
    alphaInputType=Bool
    rotationMode=BMM_Additive
    rotationSpace=BCS_ParentBoneSpace
    translationMode=BMM_Additive
    translationSpace=BCS_BoneSpace
    alphaBoolBlend {blendOption=Linear}
    boneToModify {boneName=dyn_b_hair_1}
    rotation {yaw=10}
   }
   n15 = "Transform (Modify) Bone - Bone: dyn_b_base_hair"!(ComponentPose=@n14, bAlphaBoolEnabled:bool=Get bIsSkyDiving(), Translation="0,0,-1", Rotation="0,17,0", Scale="1,1,1") -> (Pose) {
    alphaInputType=Bool
    rotationMode=BMM_Additive
    rotationSpace=BCS_BoneSpace
    translationMode=BMM_Additive
    translationSpace=BCS_BoneSpace
    alphaBoolBlend {blendOption=Linear}
    boneToModify {boneName=dyn_b_base_hair}
    rotation {yaw=17}
    translation {z=-1}
   }
   n16 = "Transform (Modify) Bone - Bone: dyn_b_base_hair"(ComponentPose=@n15, bAlphaBoolEnabled=Get bIsSkydiveFloating(), Translation="0,0,-1", Rotation="0,0,0", Scale="1,1,1") {
    alphaInputType=Bool
    rotationMode=BMM_Additive
    rotationSpace=BCS_BoneSpace
    translationMode=BMM_Additive
    translationSpace=BCS_BoneSpace
    alphaBoolBlend {blendOption=Linear}
    boneToModify {boneName=dyn_b_base_hair}
    translation {z=-1}
   }
   Save cached pose 'HairCache'(Pose=Component To Local(ComponentPose=@n16)) {cacheName=HairCache}
   n17 = RigidBody(ComponentPose=Local To Component(LocalPose=Copy Pose From Mesh() {bUseAttachedParent=true}), OverrideWorldGravity=Get GravityOverride()) {
    bOverrideWorldGravity=true
    overridePhysicsAsset=&F_MED_MeekCrow_FaceAcc_fronthair_Physics
    simulationSpace=BaseBoneSpace
    alphaBoolBlend {blendOption=Linear}
    baseBoneRef {boneName=head}
    overrideWorldGravity {z=-980}
    simSpaceSettings {maxAngularAcceleration=40, maxAngularVelocity=8, maxLinearAcceleration=1000, maxLinearVelocity=100, worldAlpha=1}
   }
   n18 = "Constraint - dyn_b_base_hair (PP)"!(ComponentPose=@n17, Alpha=1, ConstraintWeights_0=0.1, ConstraintWeights_1:float=0.9) -> (Pose) {
    alphaBoolBlend {blendOption=Linear}
    boneToModify {boneName=dyn_b_base_hair}
    constraintSetup [{targetBone {boneName=neck_01}}, {targetBone {boneName=spine_05}}]
    constraintWeights [1,1]
   }
   n19 = "Constraint - dyn_b_hair_4 (PP)"!(ComponentPose=@n18, Alpha=1, ConstraintWeights_0=0.8, ConstraintWeights_1:float=0.2) -> (Pose) {
    alphaBoolBlend {blendOption=Linear}
    boneToModify {boneName=dyn_b_hair_4}
    constraintSetup [{targetBone {boneName=spine_04}}, {targetBone {boneName=spine_05}}]
    constraintWeights [1,1]
   }
   n20 = Anim Dynamics - Bone: dyn_b_hair_1(ComponentPose=@n19, Alpha=1, ExternalForce="0,0,0") {
    simulationSpace=BoneRelative
    alphaBoolBlend {blendOption=Linear}
    boundBone {boneName=dyn_b_hair_1}
    physicsBodyDefinitions [{
      boundBone {boneName=dyn_b_hair_1}
      boxExtents=<20 20 10>
      constraintSetup {angularLimitsMax=<0 0 3>, angularLimitsMin=<0 0 -3>, angularTarget=<0 0 0>, linearAxesMax=<0.5 0 0>}
      localJointOffset=<0 0 0>
     }]
    relativeSpaceBone {boneName=head}
   }
   Save cached pose 'HairNoDYN'(Pose=Component To Local(ComponentPose=@n20)) {cacheName=HairNoDYN}
  ]}
 EventGraph {nodes [
   Event BlueprintUpdateAnimation()
   > n1 = Sequence()
   > n2 = Gravity Override State Function DEPRECATED(
    CharacterPartAnimInstance=Self-Reference()
    GameTimeInSeconds=GetGameTimeInSeconds()
    GravityOverrideParams=Get Gravity Override Params()
    RBAN Params=Get RBAN Params()
   )
   > Set GravityOverride(GravityOverride=@n2)
   > Set headrot(headrot:float=GetAlphaFromReferenceTransformDelta(SkeletalMeshComponent=GetPartSkeletalMeshComponent(), BoneName=head, BaseName=spine_05, InputRange=(X=-10,Y=-45), TwistAxis="1,0,0")) -> (Output_Get:float)
   n3 = GetOwnerSkeletalMeshComponent()
   n4 = BreakRotator(InRot=BreakTransform(InTransform=GetDeltaTransformFromRefPose(self=@n3, BoneName=head, BaseName=spine_05)).Rotation)
   > Set HairRotator(HairRotator:Rotator=MakeRotator(Yaw=MapRangeClamped(Value=@n4.Yaw, InRangeA=0, InRangeB=0, OutRangeB=0))) -> (Output_Get:Rotator)
   n5 = OR Boolean(A=Get bIsSkydiveFloating(), B=Get bIsSkydiveDiveMode())
   n6 = Reroute Node(InputPin:double) -> (OutputPin:double)
   Set isSkydiveFloating(
    execute=@n1.then_1
    isSkydiveFloating:double=FInterpTo(Current=Get isSkydiveFloating() -> (isSkydiveFloating:double), Target="To Float (Boolean)"(InBool=@n5), DeltaTime=@n6, InterpSpeed=8)
   ) -> (Output_Get:double)
   n7 = Get bIsCrouching()
   > Set isCrouching(isCrouching:double=FInterpTo(Current=Get isCrouching(), Target="To Float (Boolean)"(InBool=@n7), DeltaTime=@n6, InterpSpeed=15)) -> (Output_Get:double)
   n8 = Get bIsTargeting()
   > Set isTargeting(isTargeting:double=FInterpTo(Current=Get isTargeting() -> (isTargeting:double), Target="To Float (Boolean)"(InBool=@n8), DeltaTime=@n6, InterpSpeed=15)) -> (Output_Get:double)
   > n9 = Sequence()
   > n10 = Branch(Condition=Get bIsInFrontEnd())
   > Set BkHairDynAlpha(BkHairDynAlpha:double=0.15) -> (Output_Get:double)
   n11 = Branch(execute=@n10.else, Condition=OR Boolean!(A=Get bIsSkydiveDiveMode(), B=Get bIsSkydiveFloating(), C:bool=Get bIsCrouching()) -> (ReturnValue))
   > Set BkHairDynAlpha(BkHairDynAlpha=0.3)
   Set BkHairDynAlpha(execute=@n11.else, BkHairDynAlpha=0.8)
   n12 = Branch(execute=@n9.then_1, Condition=Get bIsInFrontEnd())
   > Set BangsDynAlpha(BangsDynAlpha:double=0.5) -> (Output_Get:double)
   > Set HatDynAlpha(HatDynAlpha:double=0.25) -> (Output_Get:double)
   > Set FnHairDynAlpha(FnHairDynAlpha:double=0.35) -> (Output_Get:double)
   Set BangsDynAlpha(execute=@n12.else, BangsDynAlpha=1)
   > Set HatDynAlpha(HatDynAlpha=1)
   > n13 = Branch(Condition=Get bIsParachuteOpened())
   > Set FnHairDynAlpha(FnHairDynAlpha=0.5)
   Set FnHairDynAlpha(execute=@n13.else, FnHairDynAlpha=1)
   n14 = Event OnCharacterCustomizationLoaded()
   > n15 = Sequence()
   > Set Skel_Mesh(Skel_Mesh=@n14.CustomBaseMesh)
   > n16 = Cast To PlayerPawn_Athena(Object=TryGetPawnOwner())
   > Set Body_Mesh(Body_Mesh=GetSkeletalMeshForPartType(self=@n16))
   n17 = CheckForBackbling()
   > Set isWearingPackBool(isWearingPackBool:bool=OR Boolean(A=@n17, B=@n17.bIsBackblingACape, C=@n17.bIsBackblingLong)) -> (Output_Get:bool)
   n18 = OR Boolean!(A=@n17.bIsBackblingACape, B=@n17.bIsBackblingLong) -> (ReturnValue)
   > Set is_f_WearingCape(is_f_WearingCape:double="To Float (Boolean)"(InBool=@n18)) -> (Output_Get:double)
   > Set isWearingBP(isWearingBP:bool=AND Boolean(A=@n17, B=NOT Boolean(A=@n18))) -> (Output_Get:bool)
   > n19 = Branch(Condition=Get bIsInFrontEnd())
   > Set Sim Space A(Sim Space A:AnimPhysSimSpaceType Enum=World) -> (Output_Get:AnimPhysSimSpaceType Enum)
   > Set Sim Space B(Sim Space B:AnimPhysSimSpaceType Enum=World) -> (Output_Get:AnimPhysSimSpaceType Enum)
   > Set Sim Space C(Sim Space C:AnimPhysSimSpaceType Enum=World) -> (Output_Get:AnimPhysSimSpaceType Enum)
   > Set Sim Space RB(Sim Space RB:Sim Space Settings Structure=Get Sim Space FE() -> (Sim Space FE:Sim Space Settings Structure)) -> (Output_Get:Sim Space Settings Structure)
   Set Sim Space A(execute=@n19.else)
   > Set Sim Space B(Sim Space B=BoneRelative)
   > Set Sim Space C(Sim Space C=Component)
   > Set Sim Space RB(Sim Space RB=Get Sim Space BR() -> (Sim Space BR:Sim Space Settings Structure))
   n20 = CheckForBackbling()
   Set isWearingPack(execute=@n15.then_1, isWearingPack:bool=OR Boolean!(A=@n20, B=@n20.bIsBackblingACape, C:bool=@n20.bIsBackblingLong) -> (ReturnValue)) -> (Output_Get:bool)
   Event OnWeaponChanged()
   > Set isGunOut(isGunOut:double=1) -> (Output_Get:double)
   > n21 = Switch on EFortWeaponCoreAnimation!(execute, Selection:EFortWeaponCoreAnimation Enum=Get WeaponCoreAnim!() -> (WeaponCoreAnim)) -> (
    Melee:Exec
    Pistol:Exec
    Shotgun:Exec
    PaperBlueprint:Exec
    Rifle:Exec
    MeleeOneHand:Exec
    MachinePistol:Exec
    RocketLauncher:Exec
    GrenadeLauncher:Exec
    GoingCommando:Exec
    AssaultRifle:Exec
    TacticalShotgun:Exec
    SniperRifle:Exec
    TrapPlacement:Exec
    ShoulderLauncher:Exec
    AbilityDecoTool:Exec
    Crossbow:Exec
    C4:Exec
    RemoteControl:Exec
    DualWield:Exec
    AR_BullPup:Exec
    AR_ForwardGrip:Exec
    MedPackPaddles:Exec
    SMG_P90:Exec
    AR_DrumGun:Exec
    Consumable_Small:Exec
    Consumable_Large:Exec
    Balloon:Exec
    MountedTurret:Exec
    CreativeTool:Exec
    ExplosiveBow:Exec
    AshtonIndigo:Exec
    AshtonChicago:Exec
    MeleeDualWield:Exec
    Unarmed:Exec
   )
   Set isGunOut(execute=[@n21.Melee, @n21.MeleeDualWield, @n21.MeleeOneHand, @n21.Unarmed], isGunOut=0)
   > Set HeadUpAlpha(HeadUpAlpha:double=GetAlphaFromReferenceTransformDelta(SkeletalMeshComponent=Get Skel_Mesh(), BoneName=head, BaseName=neck_01, InputRange=(X=5,Y=45), TwistAxis="1,0,0")) -> (Output_Get:double)
  ]}
]
variables [
 BangsDynAlpha {type=number}
 BkHairDynAlpha {type=number}
 Body_Mesh {type=SkeletalMeshComponent}
 FnHairDynAlpha {type=number}
 Gravity Override Params {type=GravityOverrideParamsStruct}
 GravityOverride {type=Vector}
 HairRotator {type=Rotator}
 HatDynAlpha {type=number}
 headrot {type=number}
 HeadUpAlpha {type=number}
 is_f_WearingCape {type=number}
 isCrouching {type=number}
 isGunOut {type=number}
 isSkydiveFloating {type=number}
 isTargeting {type=number}
 isWearingBP {type=boolean}
 isWearingPack {type=boolean}
 isWearingPackBool {type=boolean}
 RBAN Params {
  type=RBANParamsStruct
  default {
   dBNO {simSpaceSettings {
     maxAngularAcceleration=1000
     maxAngularVelocity=3
     maxLinearAcceleration=300
     maxLinearVelocity=700
     velocityScaleZ=0
     externalLinearDragV=<0 0.5 0>
    }}
   dBNO_Carried {simSpaceSettings {
     maxAngularAcceleration=1000
     maxAngularVelocity=3
     maxLinearAcceleration=300
     maxLinearVelocity=700
     velocityScaleZ=0
     externalLinearDragV=<0 0.5 0>
    }}
   emoteOrMelee {simSpaceSettings {
     maxAngularAcceleration=1000
     maxAngularVelocity=3
     maxLinearAcceleration=300
     maxLinearVelocity=700
     velocityScaleZ=0
     externalLinearDragV=<0 0.5 0>
    }}
   front_End {simSpaceSettings {
     maxAngularAcceleration=1000
     maxAngularVelocity=3
     maxLinearAcceleration=300
     maxLinearVelocity=700
     velocityScaleZ=0
     externalLinearDragV=<0 0.5 0>
    }}
   onGround_Crouch_Moving {simSpaceSettings {
     maxAngularAcceleration=1000
     maxAngularVelocity=3
     maxLinearAcceleration=300
     maxLinearVelocity=700
     velocityScaleZ=0
     externalLinearDragV=<0 0.5 0>
    }}
   onGround_Falling {simSpaceSettings {
     maxAngularAcceleration=1000
     maxAngularVelocity=3
     maxLinearAcceleration=300
     maxLinearVelocity=700
     velocityScaleZ=0
     externalLinearDragV=<0 0.5 0>
    }}
   onGround_Moving {simSpaceSettings {
     maxAngularAcceleration=1000
     maxAngularVelocity=3
     maxLinearAcceleration=300
     maxLinearVelocity=700
     velocityScaleZ=0
     externalLinearDragV=<0 0.5 0>
    }}
   onGround_Standing {simSpaceSettings {
     maxAngularAcceleration=1000
     maxAngularVelocity=3
     maxLinearAcceleration=300
     maxLinearVelocity=700
     velocityScaleZ=0
     externalLinearDragV=<0 0.5 0>
    }}
   rail_Grinding {simSpaceSettings {
     maxAngularAcceleration=1000
     maxAngularVelocity=3
     maxLinearAcceleration=300
     maxLinearVelocity=700
     velocityScaleZ=0
     worldAlpha=0.1
     externalLinearDragV=<0 0.5 0>
    }}
   riding_Moving {simSpaceSettings {
     maxAngularAcceleration=1000
     maxAngularVelocity=3
     maxLinearAcceleration=300
     maxLinearVelocity=700
     velocityScaleZ=0
     externalLinearDragV=<0 0.5 0>
    }}
   riding_Moving_Targeting {simSpaceSettings {
     maxAngularAcceleration=1000
     maxAngularVelocity=3
     maxLinearAcceleration=300
     maxLinearVelocity=700
     velocityScaleZ=0
     externalLinearDragV=<0 0.5 0>
    }}
   riding_Standing {simSpaceSettings {
     maxAngularAcceleration=1000
     maxAngularVelocity=3
     maxLinearAcceleration=300
     maxLinearVelocity=700
     velocityScaleZ=0
     externalLinearDragV=<0 0.5 0>
    }}
   skydive_DivingDown {simSpaceSettings {
     maxAngularAcceleration=1000
     maxAngularVelocity=3
     maxLinearAcceleration=300
     maxLinearVelocity=700
     velocityScaleZ=0
     externalLinearDragV=<0 0.5 0>
    }}
   skydive_FloatingDown {
    componentLinearAccScale=<0 0.5 0>
    componentLinearVelScale=<0 0.5 0>
    simSpaceSettings {
     maxAngularAcceleration=1000
     maxAngularVelocity=3
     maxLinearAcceleration=300
     maxLinearVelocity=700
     velocityScaleZ=0
     externalLinearDragV=<0 0.5 0>
    }
   }
   skydive_GoingUp {simSpaceSettings {
     maxAngularAcceleration=1000
     maxAngularVelocity=3
     maxLinearAcceleration=300
     maxLinearVelocity=700
     velocityScaleZ=0
     externalLinearDragV=<0 0.5 0>
    }}
   skydive_Parachuting {
    componentLinearAccScale=<0 0.1 0.3>
    componentLinearVelScale=<0 0.1 0.3>
    simSpaceSettings {
     maxAngularAcceleration=1000
     maxAngularVelocity=3
     maxLinearAcceleration=300
     maxLinearVelocity=700
     velocityScaleZ=0
     externalLinearDragV=<0 0.5 0>
    }
   }
   skydive_Surfing {simSpaceSettings {
     maxAngularAcceleration=1000
     maxAngularVelocity=3
     maxLinearAcceleration=300
     maxLinearVelocity=700
     velocityScaleZ=0
     externalLinearDragV=<0 0.5 0>
    }}
   swimming {simSpaceSettings {
     maxAngularAcceleration=1000
     maxAngularVelocity=3
     maxLinearAcceleration=300
     maxLinearVelocity=700
     velocityScaleZ=0
     externalLinearDragV=<0 0.5 0>
    }}
   tactical_Sprint {simSpaceSettings {
     maxAngularAcceleration=1000
     maxAngularVelocity=3
     maxLinearAcceleration=300
     maxLinearVelocity=700
     velocityScaleZ=0
     externalLinearDragV=<0 0.5 0>
    }}
   ziplining {simSpaceSettings {
     maxAngularAcceleration=1000
     maxAngularVelocity=3
     maxLinearAcceleration=300
     maxLinearVelocity=700
     velocityScaleZ=0
     externalLinearDragV=<0 0.5 0>
    }}
  }
 }
 Sim Space A {type=AnimPhysSimSpaceType}
 Sim Space B {type=AnimPhysSimSpaceType}
 Sim Space BR {
  type=SimSpaceSettings
  default {
   dampingAlpha=1
   maxAngularAcceleration=10000
   maxAngularVelocity=10000
   maxLinearAcceleration=10000
   maxLinearVelocity=10000
   velocityScaleZ=1
   worldAlpha=0
   externalLinearDragV=<0 0 0>
  }
 }
 Sim Space C {type=AnimPhysSimSpaceType}
 Sim Space FE {
  type=SimSpaceSettings
  default {
   dampingAlpha=1
   maxAngularAcceleration=10000
   maxAngularVelocity=10000
   maxLinearAcceleration=10000
   maxLinearVelocity=10000
   velocityScaleZ=1
   worldAlpha=0
   externalLinearDragV=<0 0 0>
  }
 }
 Sim Space RB {
  type=SimSpaceSettings
  default {
   dampingAlpha=1
   maxAngularAcceleration=10000
   maxAngularVelocity=10000
   maxLinearAcceleration=10000
   maxLinearVelocity=10000
   velocityScaleZ=1
   worldAlpha=0
   externalLinearDragV=<0 0 0>
  }
 }
 Skel_Mesh {type=SkeletalMeshComponent}
 Skeletal Mesh Component {type=SkinnedMeshComponent, default=null}
]
