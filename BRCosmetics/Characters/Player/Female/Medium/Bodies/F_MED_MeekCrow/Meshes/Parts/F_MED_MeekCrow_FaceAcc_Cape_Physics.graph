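# Physics asset for the cape face-accessory part of the female medium MeekCrow player body, shaping the upper body, likely as collision for the cape.
# Capsules cover the upper spine, clavicles and lower neck, while the upper arms and neck_02 carry only tiny capsules.
# No cape bones have bodies.
# Constraints chain the bones outward from the spine with all angular motion free, so the set swing and twist limits are not enforced.
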PhysicsAsset /BRCosmetics/Characters/Player/Female/Medium/Bodies/F_MED_MeekCrow/Meshes/Parts/F_MED_MeekCrow_FaceAcc_Cape_Physics
bodies {
 clavicle_l [{shapeType=Capsule, length=27.25, radius=11.005, shapeName=clavicle_l_capsule, center=<10.912 0 0>, rotation=<rot -0.0006 -90 90>}]
 clavicle_r [{shapeType=Capsule, length=27.25, radius=11.005, shapeName=clavicle_r_capsule, center=<-10.912 0 0>, rotation=<rot -0.0006 -90 -90>}]
 neck_01 [{shapeType=Capsule, length=10.255, radius=5.755, shapeName=neck_01_capsule, center=<-3.9303 1.4597 0>, rotation=<rot 0 -90 0>}]
 neck_02 [{shapeType=Capsule, length=0.505, radius=0.505, shapeName=neck_02_capsule, center=<-1.9928 -3.7649 -0.0077>, rotation=<rot 0 3.0453 -75.8876>}]
 spine_03 [{shapeType=Capsule, length=24.63, radius=15.005, shapeName=spine_03_capsule, rotation=<rot 0 -90 0>}]
 spine_04 [{shapeType=Capsule, length=33.625, radius=11.88, shapeName=spine_04_capsule, rotation=<rot 0 -90 0>}]
 spine_05 [{
   shapeType=Capsule
   length=42.3119
   radius=11.6287
   shapeName=spine_05_capsule
   center=<5.6593 0.7852 -0.1789>
   rotation=<rot 5.8056 -89.7641 -1.3091>
  }]
 upperarm_l [{shapeType=Capsule, length=0.505, radius=0.505, shapeName=upperarm_l_capsule, rotation=<rot 0 -90 0>}]
 upperarm_r [{shapeType=Capsule, length=0.505, radius=0.505, shapeName=upperarm_r_capsule, rotation=<rot 0 -90 0>}]
}
constraints [
 {
  bone1Name=spine_04
  bone2Name=spine_03
  swing1LimitDegrees=45
  swing1Motion=Free
  swing2LimitDegrees=45
  swing2Motion=Free
  twistLimitDegrees=45
  twistMotion=Free
 }
 {
  bone1Name=spine_05
  bone2Name=spine_04
  swing1LimitDegrees=45
  swing1Motion=Free
  swing2LimitDegrees=45
  swing2Motion=Free
  twistLimitDegrees=45
  twistMotion=Free
 }
 {
  bone1Name=clavicle_l
  bone2Name=spine_05
  swing1LimitDegrees=45
  swing1Motion=Free
  swing2LimitDegrees=45
  swing2Motion=Free
  twistLimitDegrees=45
  twistMotion=Free
 }
 {
  bone1Name=clavicle_r
  bone2Name=spine_05
  swing1LimitDegrees=45
  swing1Motion=Free
  swing2LimitDegrees=45
  swing2Motion=Free
  twistLimitDegrees=45
  twistMotion=Free
 }
 {
  bone1Name=upperarm_l
  bone2Name=clavicle_l
  swing1LimitDegrees=45
  swing1Motion=Free
  swing2LimitDegrees=45
  swing2Motion=Free
  twistLimitDegrees=45
  twistMotion=Free
 }
 {
  bone1Name=upperarm_r
  bone2Name=clavicle_r
  swing1LimitDegrees=45
  swing1Motion=Free
  swing2LimitDegrees=45
  swing2Motion=Free
  twistLimitDegrees=45
  twistMotion=Free
 }
 {
  bone1Name=neck_01
  bone2Name=spine_05
  swing1LimitDegrees=45
  swing1Motion=Free
  swing2LimitDegrees=45
  swing2Motion=Free
  twistLimitDegrees=45
  twistMotion=Free
 }
 {
  bone1Name=neck_02
  bone2Name=neck_01
  swing1LimitDegrees=45
  swing1Motion=Free
  swing2LimitDegrees=45
  swing2Motion=Free
  twistLimitDegrees=45
  twistMotion=Free
 }
]
